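; Parent NPC gameplay ability that fires a fan of projectiles, with CreateFanProjectile left empty, apparently for child overrides.
; On activation it plays a montage and spawns the fan on a tagged notify, on completion without one, or immediately without a montage.
; The fan centers on the avatar FortPlayerPawn's root bone plus an offset, spacing projectiles by rotating its right vector around world Z across the angle at a set radius.
; A debug flag enables prints and shapes.
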(event Ability|EventActivateAbility
  (bind (_montageinfo_montagetoplay _montageinfo_animplayrate _montageinfo_animrootmotiontranslationscale) (Utilities|Struct|BreakFortGameplayAbilityMontageInfo (Variables|GetMontageInfo_MontageToPlay)))
  (|CallParentFunction)
  (Development|PrintString "@@@ Event ActivateAbility [GA_ProjectileFan_Parent]: CALLED!" false (Variables|Debug|GetbDebugAbility))
  (Ability|CommitAbility)
  (Variables|InternalData|SetbCaughtNotify false)
  (Utilities|IsValid _montageinfo_montagetoplay
    (:Is%20Valid
      (bind (_asynctaskproxy _notifytag) (Ability|Tasks|PlayMontageAndWaitNotify "" _montageinfo_montagetoplay _montageinfo_animplayrate "None" true _montageinfo_animrootmotiontranslationscale)
        (:OnCompleted
          (if (not (Variables|InternalData|GetbCaughtNotify))
            (CallFunction|SpawnProjectileFan :!label _label)))
        (:OnInterrupted
          _)
        (:OnCancelled
          _)
        (:OnNotifyBegin
          (if (== _notifytag (Variables|Tunables|GetMontageNotifyTag))
            (Variables|InternalData|SetbCaughtNotify true)
            (goto _label))))
      (Ability|EndAbility))
    (:Is%20Not%20Valid
      (CallFunction|SpawnProjectileFan))))

(event Custom|SpawnProjectileFan
  (bind _fancenterrotator (Variables|InternalData|GetFanCenterRotator))
  (bind _returnvalue (Math|Vector|GetRightVector _fancenterrotator))
  (bind _angle (Variables|Tunables|GetAngle))
  (bind _returnvalue_2 (/ _angle 2))
  (bind _incrementangle (Variables|InternalData|GetIncrementAngle))
  (bind _incrementrotator (Variables|InternalData|GetIncrementRotator))
  (bind (_fireoffset_x _fireoffset_y _fireoffset_z) (Math|Vector|BreakVector (Variables|GetFireOffset_X)))
  (Development|PrintString "@@@@@@ Event SpawnProjectileFan [GA_ProjectileFan_Parent]: CALLED!" false (Variables|Debug|GetbDebugAbility))
  (Utilities|FlowControl|Sequence
    (:then_0
      (bind (_fanlocation _fanrotator) (CallFunction|GetFanPointData))
      (Variables|InternalData|SetFanCenterRotator _fanrotator)
      (bind _output_get (Variables|InternalData|SetFanCenterLoc (Utilities|Operators|Add _fanlocation (* (Math|Vector|GetForwardVector _fanrotator) _fireoffset_x) (* (Math|Vector|GetRightVector _fanrotator) _fireoffset_y) (* (Math|Vector|GetUpVector _fanrotator) _fireoffset_z))))
      (if (Variables|Debug|GetbDebugAbility)
        (Rendering|Debug|DrawDebugSphere _output_get 5e+01 12 "(R=1.000000,G=0.027003,B=0.636845,A=1.000000)" 1e+01 2.0)))
    (:then_1
      (Variables|InternalData|SetIncrementAngle (/ (Variables|Tunables|GetAngle) (Variables|Tunables|GetProjectileCount)))
      (Variables|InternalData|SetIncrementRotator (/ (Variables|InternalData|GetIncrementAngle) 2))
      (if (Variables|Debug|GetbDebugAbility)
        (Rendering|Debug|DrawDebugLine (Variables|InternalData|GetFanCenterLoc) (+ (Variables|InternalData|GetFanCenterLoc) (* (Math|Vector|GetRightVector (Variables|InternalData|GetFanCenterRotator)) 5e+02)) "(R=0.458637,G=0.025770,B=0.989583,A=1.000000)" 1e+01 3.0))
      (for _index (range 1 (+ (Variables|Tunables|GetProjectileCount) 1))
        (Variables|InternalData|SetCurrentProjectileRot (Math|Rotator|MakeRotfromX (Math|Vector|RotateVectorAroundAxis _returnvalue (Utilities|Operators|Subtract (* _incrementangle _index) _incrementrotator _returnvalue_2) "0.000000,0.000000,1.000000")))
        (Variables|InternalData|SetCurrentProjectilePos (+ (Variables|InternalData|GetFanCenterLoc) (* (Math|Vector|RotateVectorAroundAxis _returnvalue (Utilities|Operators|Subtract (* _incrementangle _index) _incrementrotator _returnvalue_2) "0.000000,0.000000,1.000000") (Variables|Tunables|GetRadius))))
        (if (Variables|Debug|GetbDebugAbility)
          (Rendering|Debug|DrawDebugSphere (Variables|InternalData|GetCurrentProjectilePos) 5e+01 12 "(R=0.005771,G=0.489583,B=0.340933,A=1.000000)" 1e+01 2.0))
        (CallFunction|CreateFanProjectile))
      (if (not (Variables|InternalData|GetbCaughtNotify))
        (Ability|EndAbility)))))

(event Ability|EventOnEndAbility
  (params (bWasCancelled :type Boolean))
  (Development|PrintString "@@@@@@@@@ Event OnEndAbility [GA_ProjectileFan_Parent]: CALLED!" false (Variables|Debug|GetbDebugAbility))
  (|CallParentFunction bWasCancelled))

(fn CreateFanProjectile)

(fn GetFanPointData
  (bind _asfort_player_pawn (Utilities|Casting|CastToFortPlayerPawn (Ability|GetAvatarActorfromActorInfo))
    (:then
      (bind (_returnvalue_location _returnvalue_rotation) (Math|Transform|BreakTransform (Transformation|GetBoneTransform (Variables|/Script/Engine.Character|GetMesh _asfort_player_pawn) "Root" "RTS_World")))
      (return _returnvalue_location _returnvalue_rotation))
    (:CastFailed
      (Ability|EndAbility))))
